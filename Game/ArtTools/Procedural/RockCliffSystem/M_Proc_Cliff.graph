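# Procedural cliff material for instanced and Nanite meshes that blends a tinted rock surface, with world-aligned detail normals, into the mobile landscape grass material.
# A height-lerp mask combines local-Y falloff, wavered by a world-position sine, with upward-facing normals and the grass height, plus an optional IsRamp gradient.
# The transition also darkens rock color and specular.
# Cave and NoCliff switches output pure rock or pure landscape.
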Material /Game/ArtTools/Procedural/RockCliffSystem/M_Proc_Cliff
graph {
 properties {
  bTangentSpaceNormal=false
  bUseMaterialAttributes=true
  bUsedWithInstancedStaticMeshes=true
  bUsedWithNanite=true
  pixelDepthOffsetMode=PDOM_Legacy
  usedShadingModels=Default Lit
 }
 nodes [
  n1 = WorldPosition()
  n2 = Transform(None=Subtract(A=@n1.XYZ, B=ObjectPositionWS())) {transformSourceType=TRANSFORMSOURCE_World, transformType=TRANSFORM_Local}
  n3 = Abs(None=ComponentMask(None=@n2) {r=true})
  n4 = Multiply(
   A=ScalarParameter() {defaultValue=64, parameterName=AssetWidth}
   B=MaterialFunctionCall() {materialFunction=&ObjectScale}.Scale X
  )
  n5 = ComponentMask(None=@n2) {g=true}
  n6 = Add(A=@n5, B=ScalarParameter() {defaultValue=-256, parameterName=FalloffBias})
  n7 = ComponentMask(None=@n1.XYZ) {r=true}
  n8 = ComponentMask(None=@n1.XYZ) {g=true}
  n9 = Sine(None=Divide(A=Add(A=@n7, B=@n8), B=ScalarParameter() {defaultValue=1200, parameterName=Sine}))
  n10 = LinearInterpolate(
   A=ScalarParameter() {defaultValue=16, parameterName=FalloffOffset}
   B=ScalarParameter() {defaultValue=64, parameterName=FalloffOffset2}
   Alpha=@n9
  )
  n11 = Saturate(None=Divide(A=Subtract(A=@n6, B=@n10), B=ScalarParameter() {defaultValue=-256, parameterName=FalloffDistance}))
  Add(A=Saturate(None=Divide(A=@n3, B=@n4)), B=@n11)
  Multiply()
  ObjectRadius()
  Saturate(None=Divide(A=Add(A=@n9)))
  ScalarParameter() {defaultValue=0.5, parameterName=BumpOffset}
  n12 = MaterialFunctionCall!(CombinedLandscapeLayersMask, BypassUseRVT, BypassBiomeColors) {materialFunction=/Game/Athena/Apollo/Environments/Landscape/Materials/Biome/CommonMaterials/Mobile/MF_Mobile_Common_Grass}
  n13 = TextureSample(UVs=Multiply(A=TextureCoordinate(), B=ScalarParameter() {defaultValue=0.8913, parameterName=DetailScale})) {texture=/Game/Environments/Apollo/Rocks/Natural_Rocks/Textures/T_RockDetails_02D}
  n14 = TextureCoordinate()
  n15 = TextureSampleParameter2D(UVs=@n14) {parameterName=Diffuse, texture=/Game/Environments/Apollo/Sets/Terrain/Textures/T_Apollo_Cliff_D_01}
  n16 = TextureSampleParameter2D(UVs=@n14) {parameterName=Mask, samplerType=SAMPLERTYPE_Masks, texture=/Game/Environments/Apollo/Sets/Terrain/Textures/T_Apollo_Cliff_S_01}
  n17 = TextureSampleParameter2D(UVs=@n14) {parameterName=Normal, samplerType=SAMPLERTYPE_Normal, texture=/Game/Environments/Apollo/Sets/Terrain/Textures/T_Apollo_Cliff_N_01}
  n18 = ScalarParameter() {defaultValue=500, parameterName=DetailNormalScale}
  n19 = LinearInterpolate(A=ScalarParameter() {defaultValue=-1.3885, parameterName=C1}, B=ScalarParameter() {defaultValue=2.7646, parameterName=C2}, Alpha=@n15.A)
  n20 = MaterialFunctionCall(
   Normal=MaterialFunctionCall(
    TextureObject=TextureObject() {samplerType=SAMPLERTYPE_Normal, texture=/Game/Environments/Apollo/Sets/Terrain/Textures/T_Apollo_Cliff_N_03}
    TextureSize=@n18
   ) {materialFunction=&WorldAlignedNormal}.XYZ Texture
   Flatness=LinearInterpolate(B=ScalarParameter() {defaultValue=0.5744, parameterName=DetailNormal}, Alpha=Saturate(None=@n19)) {constA=1}
  ) {materialFunction=&FlattenNormal}
  n21 = SetMaterialAttributes!(
   MaterialAttributes
   Base Color=Multiply(A=VectorParameter() {parameterName=CliffColor, defaultValue=<rgba 0.75 0.682 0.6602 1>}.RGB, B=Multiply(A=@n13.RGB, B=@n15.RGB))
   Metallic
   Specular=@n16.R
   Roughness=@n16.R
   Anisotropy
   Emissive Color
   Opacity
   Opacity Mask
   Normal=Transform(None=MaterialFunctionCall(BaseNormal=@n17.RGB, AdditionalNormal=@n20.Result) {materialFunction=&BlendAngleCorrectedNormals}.Result)
  ) {attributeSetTypes [
    #1
    #2
    #3
    #4
    #5
    #6
    #7
    #8
    #9
   ]}
  n22 = GetMaterialAttributes(None=@n21) {attributeGetTypes [#1, #3]}
  n23 = ScalarParameter() {defaultValue=1, parameterName=A}
  n24 = ScalarParameter() {defaultValue=-1.9608, parameterName=B}
  n25 = Saturate(None=Divide(A=@n5, B=ScalarParameter() {defaultValue=-512, parameterName=FalloffDistanceRamp}))
  n26 = ScalarParameter() {defaultValue=-0.5, parameterName=UpVectorBlend_Low}
  n27 = ScalarParameter() {defaultValue=1, parameterName=UpVectorBlend_High}
  n28 = ComponentMask(None=VertexNormalWS()) {b=true}
  n29 = Multiply(A=Saturate(None=@n11), B=Saturate(None=LinearInterpolate(A=@n26, B=@n27, Alpha=@n28)))
  n30 = MaterialFunctionCall(
   Transition Phase=StaticSwitchParameter(True=Add(A=@n25, B=@n29), False=@n29) {parameterName=IsRamp}
   Height Texture=@n12.Height
   Contrast=ScalarParameter() {parameterName=ContrastC1}
  ) {materialFunction=&HeightLerp}
  n31 = LinearInterpolate(
   A=ScalarParameter() {defaultValue=0.4405, parameterName=Shadow}
   Alpha=Saturate(None=LinearInterpolate(A=@n23, B=@n24, Alpha=@n30.Lerp Alpha No Contrast))
  )
  n32 = SetMaterialAttributes!(MaterialAttributes=@n22.MaterialAttributes, Base Color=Multiply(A=@n22.Base Color, B=@n31), Specular=Multiply(A=@n22.Specular, B=@n31)) {attributeSetTypes [#1, #3]}
  StaticSwitchParameter(
   True=@n12.MobileLandscapeMaterial
   False=StaticSwitchParameter(True=@n21, False=BlendMaterialAttributes(A=@n32, B=@n12.MobileLandscapeMaterial, Alpha=@n30.Alpha) {vertexAttributeBlendType=UseA}) {parameterName=Cave}
  ) {parameterName=NoCliff}
  TextureSample(UVs=Multiply(A=TextureCoordinate(), B=@n18)) {samplerType=SAMPLERTYPE_Normal, texture=&T_Apollo_Cliff_N_03}
 ]
}
