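; ProceduralRockCliff is a grid-based procedural cliff actor that resizes one static mesh to join neighboring cliffs.
; Its construction script snaps location and yaw to the Fortnite building grid, converts actor scale into integer Width/Height, and sizes an editor guide.
; Four line traces detect inner/outer connections at each side, which lengthen or shorten and offset the mesh.
; Detected neighbors refresh with propagation disabled, limiting updates to one hop.
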(event EventBeginPlay)

(event Collision|EventActorBeginOverlap
  (params (OtherActor :type Actor%20Object%20Reference)))

(event EventTick
  (params (DeltaSeconds :type Float%20%28single-precision%29)))

(fn CheckForNeighbors
  (params (UpdateNeighbors :type Boolean))
  (Utilities|FlowControl|Sequence
    (:then_0
      (Variables|Default|SetHasRightInnerConnection false)
      (bind _outhits (Collision|MultiLineTraceForObjects :Start (Math|Transform|TransformLocation (Transformation|GetActorTransform) (+ (Math|Vector|MakeVector (/ (* (Variables|Default|GetWidth) 512.0) 2.0) 8.0 -64.0) "1.000000, 0, 0")) :End (Math|Transform|TransformLocation (Transformation|GetActorTransform) (- (Math|Vector|MakeVector (/ (* (Variables|Default|GetWidth) 512.0) 2.0) 8.0 -64.0) (Math|Vector|MakeVector (+ (/ (* (Variables|Default|GetWidth) 512.0) 2.0) 32.0)))) :ObjectTypes (Utilities|Array|MakeArray "ObjectTypeQuery3") :bTraceComplex false :DrawDebugType "ForDuration" :bIgnoreSelf true :TraceColor "(R=1.000000,G=0.000000,B=0.000000,A=1.000000)" :TraceHitColor "(R=0.000000,G=1.000000,B=0.000000,A=1.000000)" :DrawTime 5.0))
      (bind _array_element (Utilities|Array|ForEachLoopwithBreak _outhits)
        (:LoopBody
          (if (Utilities|Equal(Class) (Utilities|GetClass (.HitActor (Collision|BreakHitResult _array_element))) "/Game/ArtTools/Procedural/RockCliffSystem/ProceduralRockCliff.ProceduralRockCliff_C")
            (Utilities|FlowControl|Sequence
              (:then_0
                _)
              (:then_1
                _))
            (Variables|Default|SetHasRightInnerConnection true)
            (if UpdateNeighbors
              (bind _asprocedural_rock_cliff (Utilities|Casting|CastToProceduralRockCliff (.HitActor (Collision|BreakHitResult _array_element))))
              (CallFunction|SetStaticmeshScaleandPosition _asprocedural_rock_cliff false))))))
    (:then_1
      (Variables|Default|SetHasLeftInnerConnection false)
      (bind _outhits_2 (Collision|MultiLineTraceForObjects :Start (Math|Transform|TransformLocation (Transformation|GetActorTransform) (+ (Math|Vector|MakeVector (/ (* (Variables|Default|GetWidth) 512.0) 2.0) 8.0 -64.0) "-1.000000, 0, 0")) :End (Math|Transform|TransformLocation (Transformation|GetActorTransform) (+ (Math|Vector|MakeVector (/ (* (Variables|Default|GetWidth) 512.0) 2.0) 8.0 -64.0) (Math|Vector|MakeVector (+ (/ (* (Variables|Default|GetWidth) 512.0) 2.0) 32.0)))) :ObjectTypes (Utilities|Array|MakeArray "ObjectTypeQuery3") :bTraceComplex false :DrawDebugType "ForDuration" :bIgnoreSelf true :TraceColor "(R=1.000000,G=0.000000,B=0.000000,A=1.000000)" :TraceHitColor "(R=0.000000,G=1.000000,B=0.000000,A=1.000000)" :DrawTime 5.0))
      (bind _array_element_2 (Utilities|Array|ForEachLoopwithBreak _outhits_2)
        (:LoopBody
          (if (Utilities|Equal(Class) (Utilities|GetClass (.HitActor (Collision|BreakHitResult _array_element_2))) "/Game/ArtTools/Procedural/RockCliffSystem/ProceduralRockCliff.ProceduralRockCliff_C")
            (Utilities|FlowControl|Sequence
              (:then_0
                _)
              (:then_1
                _))
            (Variables|Default|SetHasLeftInnerConnection true)
            (if UpdateNeighbors
              (bind _asprocedural_rock_cliff_2 (Utilities|Casting|CastToProceduralRockCliff (.HitActor (Collision|BreakHitResult _array_element_2))))
              (CallFunction|SetStaticmeshScaleandPosition _asprocedural_rock_cliff_2 false))))))
    (:then_2
      (Variables|Default|SetHasRightOuterConnection false)
      (bind _outhits_3 (Collision|MultiLineTraceForObjects :Start (Math|Transform|TransformLocation (Transformation|GetActorTransform) (+ (Math|Vector|MakeVector (/ (* (Variables|Default|GetWidth) 512.0) 2.0) -272.0 -64.0) "1.000000, 0, 0")) :End (Math|Transform|TransformLocation (Transformation|GetActorTransform) (- (Math|Vector|MakeVector (/ (* (Variables|Default|GetWidth) 512.0) 2.0) -272.0 -64.0) (Math|Vector|MakeVector (+ (/ (* (Variables|Default|GetWidth) 512.0) 2.0) 32.0)))) :ObjectTypes (Utilities|Array|MakeArray "ObjectTypeQuery3") :bTraceComplex false :DrawDebugType "ForDuration" :bIgnoreSelf true :TraceColor "(R=1.000000,G=0.000000,B=0.000000,A=1.000000)" :TraceHitColor "(R=0.000000,G=1.000000,B=0.000000,A=1.000000)" :DrawTime 5.0))
      (bind _array_element_3 (Utilities|Array|ForEachLoopwithBreak _outhits_3)
        (:LoopBody
          (if (Utilities|Equal(Class) (Utilities|GetClass (.HitActor (Collision|BreakHitResult _array_element_3))) "/Game/ArtTools/Procedural/RockCliffSystem/ProceduralRockCliff.ProceduralRockCliff_C")
            (Utilities|FlowControl|Sequence
              (:then_0
                _)
              (:then_1
                _))
            (Variables|Default|SetHasRightOuterConnection true)
            (if UpdateNeighbors
              (bind _asprocedural_rock_cliff_3 (Utilities|Casting|CastToProceduralRockCliff (.HitActor (Collision|BreakHitResult _array_element_3))))
              (CallFunction|SetStaticmeshScaleandPosition _asprocedural_rock_cliff_3 false))))))
    (:then_3
      (Variables|Default|SetHasLeftOuterConnection false)
      (bind _outhits_4 (Collision|MultiLineTraceForObjects :Start (Math|Transform|TransformLocation (Transformation|GetActorTransform) (+ (Math|Vector|MakeVector (/ (* (Variables|Default|GetWidth) 512.0) 2.0) -272.0 -64.0) "-1.000000, 0, 0")) :End (Math|Transform|TransformLocation (Transformation|GetActorTransform) (+ (Math|Vector|MakeVector (/ (* (Variables|Default|GetWidth) 512.0) 2.0) -272.0 -64.0) (Math|Vector|MakeVector (+ (/ (* (Variables|Default|GetWidth) 512.0) 2.0) 32.0)))) :ObjectTypes (Utilities|Array|MakeArray "ObjectTypeQuery3") :bTraceComplex false :DrawDebugType "ForDuration" :bIgnoreSelf true :TraceColor "(R=1.000000,G=0.000000,B=0.000000,A=1.000000)" :TraceHitColor "(R=0.000000,G=1.000000,B=0.000000,A=1.000000)" :DrawTime 5.0))
      (bind _array_element_4 (Utilities|Array|ForEachLoopwithBreak _outhits_4)
        (:LoopBody
          (if (Utilities|Equal(Class) (Utilities|GetClass (.HitActor (Collision|BreakHitResult _array_element_4))) "/Game/ArtTools/Procedural/RockCliffSystem/ProceduralRockCliff.ProceduralRockCliff_C")
            (Utilities|FlowControl|Sequence
              (:then_0
                _)
              (:then_1
                _))
            (Variables|Default|SetHasLeftOuterConnection true)
            (if UpdateNeighbors
              (bind _asprocedural_rock_cliff_4 (Utilities|Casting|CastToProceduralRockCliff (.HitActor (Collision|BreakHitResult _array_element_4))))
              (CallFunction|SetStaticmeshScaleandPosition _asprocedural_rock_cliff_4 false))))))))

(fn SetStaticmeshScaleAndPosition
  (params (UpdateNeighbors :type Boolean))
  (bind _hasleftouterconnection (Variables|Default|GetHasLeftOuterConnection))
  (bind _returnvalue 512.0)
  (bind _returnvalue_2 (- (- (.x (.Max (Components|StaticMesh|GetLocalBounds (Variables|Default|GetStaticMesh)))) (.x (Components|StaticMesh|GetLocalBounds (Variables|Default|GetStaticMesh)))) _returnvalue))
  (bind _returnvalue_3 (/ _returnvalue_2 2.0))
  (bind _returnvalue_4 (+ 256.0 _returnvalue_3))
  (bind _returnvalue_5 (Math|Float|SelectFloat _returnvalue_4 0.0 _hasleftouterconnection))
  (bind _hasrightouterconnection (Variables|Default|GetHasRightOuterConnection))
  (bind _returnvalue_6 (Math|Float|SelectFloat 256.0 0.0 _hasrightouterconnection))
  (bind _width (Variables|Default|GetWidth))
  (bind _returnvalue_7 (Math|Conversions|ToFloat(Integer) _width))
  (bind _returnvalue_8 512.0)
  (bind _returnvalue_9 (* _returnvalue_8 _returnvalue_7))
  (bind _returnvalue_10 (- _returnvalue_9 _returnvalue_6))
  (bind _returnvalue_11 (- _returnvalue_10 _returnvalue_5))
  (bind _hasleftinnerconnection (Variables|Default|GetHasLeftInnerConnection))
  (bind _returnvalue_12 (Math|Float|SelectFloat 256.0 0.0 _hasleftinnerconnection))
  (bind _staticmesh (Variables|Default|GetStaticMesh))
  (bind (_min _max) (Components|StaticMesh|GetLocalBounds _staticmesh))
  (bind _hasrightinnerconnection (Variables|Default|GetHasRightInnerConnection))
  (bind _returnvalue_13 (Math|Float|SelectFloat 256.0 0.0 _hasrightinnerconnection))
  (bind _x (.x _min))
  (bind _x_2 (.x _max))
  (bind _returnvalue_14 (- _x_2 _x))
  (bind _returnvalue_15 (+ _returnvalue_11 _returnvalue_13))
  (bind _returnvalue_16 (+ _returnvalue_15 _returnvalue_12))
  ;[1 1 1] Scale Staticmesh
  (CallFunction|CheckforNeighbors UpdateNeighbors)
  (Transformation|SetRelativeScale3D (Variables|Default|GetStaticMesh) (Math|Vector|MakeVector (/ _returnvalue_16 _returnvalue_8) 1.0 (Math|Conversions|ToFloat(Integer) (Variables|Default|GetHeight))))
  (Transformation|SetRelativeLocation (Variables|Default|GetStaticMesh) (Math|Vector|vector+vector (Math|Vector|MakeVector (/ _returnvalue_16 2.0)) (Math|Vector|SelectVector "-256.000000, 0, 0" "0, 0, 0" (Variables|Default|GetHasRightInnerConnection)) (Math|Vector|SelectVector "0, 0, 0" "0, 0, 0" (Variables|Default|GetHasLeftInnerConnection)) (Math|Vector|SelectVector "256.000000, 0, 0" "0, 0, 0" (Variables|Default|GetHasRightOuterConnection)) (Math|Vector|SelectVector "0, 0, 0" "0, 0, 0" (Variables|Default|GetHasLeftOuterConnection))) false false))

(fn ConstructionScript
  (bind _returnvalue (Transformation|GetActorScale3D))
  (bind _editoronlyguide (Variables|Default|GetEditorOnlyGuide))
  (bind _width (Variables|Default|GetWidth))
  (bind _height (Variables|Default|GetHeight))
  (Utilities|FlowControl|Sequence
    (:then_0
      (Transformation|SetActorLocation (Fortnite|SnaptoFortBuildingGrid (Transformation|GetActorLocation) "0, 0, 0" "256.000000,256.000000,0.000000"))
      (Transformation|SetActorRotation (Math|Rotator|MakeRotator 0.0 0.0 (Math|Float|SnapToGrid(Float) (.yaw (Transformation|GetActorRotation)) 9e+01)))
      (Variables|Default|SetWidth (Math|Integer|Max(Integer) (Math|Float|Round (* (Variables|Default|GetWidth) (.x _returnvalue))) 1))
      (Variables|Default|SetHeight (Math|Integer|Max(Integer) (Math|Float|Round (* (Variables|Default|GetHeight) (.z _returnvalue))) 1))
      (Transformation|SetActorScale3D "1.000000,1.000000,1.000000")
      (Transformation|SetRelativeScale3D _editoronlyguide (Math|Vector|MakeVector (* _width 2.0) 1.0 (* _height 1.5)))
      (Transformation|SetRelativeLocation _editoronlyguide (Math|Vector|MakeVector (/ (* 512.0 (Math|Conversions|ToFloat(Integer) _width)) 2.0) -128.0 (/ (* 384.0 (Math|Conversions|ToFloat(Integer) _height)) -2.0)) false false))
    (:then_1
      (CallFunction|SetStaticmeshScaleandPosition true))))
